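# Physics asset for the Alien Fishhead melee pickaxe, with capsule bodies on the head, fin and tentacle bones.
# Constraints form chains from the main body: the head carries a top fin and a three-segment tip, and tentacle_base carries three ten-segment tentacles.
# Angular limits tighten in twist along the fin and tip, while the tentacle base rotates freely and tentacle A's root has one free swing axis.
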PhysicsAsset /BRCosmetics/Weapons/FORT_Melee/Pickaxe_AlienFishHead/Meshes/Pickaxe_Alien_Fishhead_Physics
bodies {
 fin_top_01 [{shapeType=Capsule, length=2.5, radius=1.755}]
 fin_top_02 [{shapeType=Capsule, length=3.25, radius=1.5429}]
 fin_top_03 [{shapeType=Capsule, length=2.75, radius=1.505}]
 head [{shapeType=Capsule, length=0.75, radius=2.25}]
 head_tip_01 [{shapeType=Capsule, length=1, radius=3, center=<-0.005 0 0>, rotation=<rot 90 0 0>}]
 head_tip_02 [{shapeType=Capsule, length=0, radius=1.5, center=<-0.005 0 0>, rotation=<rot 90 0 0>}]
 head_tip_03 [{shapeType=Capsule, length=0.25, radius=0.75, center=<-0.001 0 0>, rotation=<rot 90 0 0>}]
 main [{shapeType=Capsule, length=1, radius=3}]
 tentacleA_01 [{shapeType=Capsule, length=5.1808, radius=3.7597, rotation=<rot 90 0 0>}]
 tentacleA_02 [{shapeType=Capsule, length=4.8847, radius=4.328, rotation=<rot 90 0 0>}]
 tentacleA_03 [{shapeType=Capsule, length=5.1765, radius=5.1097, rotation=<rot 90 0 0>}]
 tentacleA_04 [{shapeType=Capsule, length=4.2559, radius=1.4416, rotation=<rot 90 0 0>}]
 tentacleA_05 [{shapeType=Capsule, length=5.2919, radius=4.347, rotation=<rot 90 0 0>}]
 tentacleA_06 [{shapeType=Capsule, length=3.3051, radius=3.0811, rotation=<rot 90 0 0>}]
 tentacleA_07 [{shapeType=Capsule, length=3.0525, radius=2.433, rotation=<rot 90 0 0>}]
 tentacleA_08 [{shapeType=Capsule, length=2.7432, radius=2.1991, rotation=<rot 90 0 0>}]
 tentacleA_09 [{shapeType=Capsule, length=2.8373, radius=1.9653, rotation=<rot 90 0 0>}]
 tentacleA_10 [{shapeType=Capsule, length=0.505, radius=0.505, rotation=<rot 90 0 0>}]
 tentacleB_01 [{shapeType=Capsule, length=5.6645, radius=3.6346, rotation=<rot 90 0 0>}]
 tentacleB_02 [{shapeType=Capsule, length=4.0013, radius=2.8243, center=<2.0513 -0.6404 -1.9197>, rotation=<rot 15.5279 -21.4717 -90>}]
 tentacleB_03 [{shapeType=Capsule, length=6.5967, radius=4.0191, center=<-0.1 0 0>, rotation=<rot 90 0 0>}]
 tentacleB_04 [{shapeType=Capsule, length=0.505, radius=0.505, rotation=<rot 90 0 0>}]
 tentacleB_05 [{shapeType=Capsule, length=5.9379, radius=3.4017, center=<-0.1 0 0>, rotation=<rot 90 0 0>}]
 tentacleB_06 [{shapeType=Capsule, length=3.9554, radius=3.3025, center=<-0.1 0 0>, rotation=<rot 90 0 0>}]
 tentacleB_07 [{shapeType=Capsule, length=2.7571, radius=3.2007, center=<-0.1 0 0>, rotation=<rot 90 0 0>}]
 tentacleB_08 [{shapeType=Capsule, length=2.4086, radius=2.8911, center=<-0.1 0 0>, rotation=<rot 90 0 0>}]
 tentacleB_09 [{shapeType=Capsule, length=3.3387, radius=2.5029, center=<-0.1 0 0>, rotation=<rot 90 0 0>}]
 tentacleB_10 [{shapeType=Capsule, length=1.5, radius=2.255, center=<-0.1 0 0>, rotation=<rot 90 0 0>}]
 tentacleC_01 [{shapeType=Capsule, length=5.1808, radius=3.7546, rotation=<rot 90 0 0>}]
 tentacleC_02 [{shapeType=Capsule, length=4.9571, radius=4.2834, rotation=<rot 90 0 0>}]
 tentacleC_03 [{shapeType=Capsule, length=6.581, radius=3.8885, rotation=<rot 90 0 0>}]
 tentacleC_04 [{shapeType=Capsule, length=4.8704, radius=3.594, rotation=<rot 90 0 0>}]
 tentacleC_05 [{shapeType=Capsule, length=4.5782, radius=3.4049, center=<-0.1 0 0>, rotation=<rot 90 0 0>}]
 tentacleC_06 [{shapeType=Capsule, length=3.4384, radius=3.1763, center=<-0.1 0 0>, rotation=<rot 90 0 0>}]
 tentacleC_07 [{shapeType=Capsule, length=2.0069, radius=3.2047, center=<-0.1 0 0>, rotation=<rot 90 0 0>}]
 tentacleC_08 [{shapeType=Capsule, length=2.2432, radius=2.8733, center=<-0.1 0 0>, rotation=<rot 90 0 0>}]
 tentacleC_09 [{shapeType=Capsule, length=2.5873, radius=2.4406, center=<-0.1 0 0>, rotation=<rot 90 0 0>}]
 tentacleC_10 [{shapeType=Capsule, length=1.5, radius=2.255, center=<-0.1 0 0>, rotation=<rot 90 0 0>}]
 tentacle_base [{shapeType=Capsule, length=0.505, radius=0.505, rotation=<rot 0 -90 0>}]
}
constraints [
 {bone1Name=head, bone2Name=main, swing1LimitDegrees=45, swing2LimitDegrees=45, twistLimitDegrees=45}
 {bone1Name=tentacleC_01, bone2Name=tentacle_base, swing1LimitDegrees=45, swing2LimitDegrees=45, twistLimitDegrees=45}
 {bone1Name=tentacleB_01, bone2Name=tentacle_base, swing1LimitDegrees=45, swing2LimitDegrees=45, twistLimitDegrees=45}
 {bone1Name=tentacleA_01, bone2Name=tentacle_base, swing1LimitDegrees=45, swing1Motion=Free, swing2LimitDegrees=45, twistLimitDegrees=45}
 {
  bone1Name=tentacle_base
  bone2Name=main
  swing1LimitDegrees=45
  swing1Motion=Free
  swing2LimitDegrees=45
  swing2Motion=Free
  twistLimitDegrees=45
  twistMotion=Free
 }
 {bone1Name=tentacleB_02, bone2Name=tentacleB_01, swing1LimitDegrees=45, swing2LimitDegrees=45, twistLimitDegrees=45}
 {bone1Name=tentacleB_03, bone2Name=tentacleB_02, swing1LimitDegrees=45, swing2LimitDegrees=45, twistLimitDegrees=45}
 {bone1Name=tentacleB_04, bone2Name=tentacleB_03, swing1LimitDegrees=45, swing2LimitDegrees=45, twistLimitDegrees=45}
 {bone1Name=tentacleB_05, bone2Name=tentacleB_04, swing1LimitDegrees=45, swing2LimitDegrees=45, twistLimitDegrees=45}
 {bone1Name=tentacleB_06, bone2Name=tentacleB_05, swing1LimitDegrees=45, swing2LimitDegrees=45, twistLimitDegrees=45}
 {bone1Name=tentacleB_07, bone2Name=tentacleB_06, swing1LimitDegrees=45, swing2LimitDegrees=45, twistLimitDegrees=45}
 {bone1Name=tentacleB_08, bone2Name=tentacleB_07, swing1LimitDegrees=45, swing2LimitDegrees=45, twistLimitDegrees=45}
 {bone1Name=tentacleB_09, bone2Name=tentacleB_08, swing1LimitDegrees=45, swing2LimitDegrees=45, twistLimitDegrees=45}
 {bone1Name=tentacleB_10, bone2Name=tentacleB_09, swing1LimitDegrees=45, swing2LimitDegrees=45, twistLimitDegrees=45}
 {bone1Name=tentacleC_02, bone2Name=tentacleC_01, swing1LimitDegrees=45, swing2LimitDegrees=45, twistLimitDegrees=45}
 {bone1Name=tentacleC_03, bone2Name=tentacleC_02, swing1LimitDegrees=45, swing2LimitDegrees=45, twistLimitDegrees=45}
 {bone1Name=tentacleC_04, bone2Name=tentacleC_03, swing1LimitDegrees=45, swing2LimitDegrees=45, twistLimitDegrees=45}
 {bone1Name=tentacleC_05, bone2Name=tentacleC_04, swing1LimitDegrees=45, swing2LimitDegrees=45, twistLimitDegrees=45}
 {bone1Name=tentacleC_06, bone2Name=tentacleC_05, swing1LimitDegrees=45, swing2LimitDegrees=45, twistLimitDegrees=45}
 {bone1Name=tentacleC_07, bone2Name=tentacleC_06, swing1LimitDegrees=45, swing2LimitDegrees=45, twistLimitDegrees=45}
 {bone1Name=tentacleC_08, bone2Name=tentacleC_07, swing1LimitDegrees=45, swing2LimitDegrees=45, twistLimitDegrees=45}
 {bone1Name=tentacleC_09, bone2Name=tentacleC_08, swing1LimitDegrees=45, swing2LimitDegrees=45, twistLimitDegrees=45}
 {bone1Name=tentacleC_10, bone2Name=tentacleC_09, swing1LimitDegrees=45, swing2LimitDegrees=45, twistLimitDegrees=45}
 {bone1Name=tentacleA_02, bone2Name=tentacleA_01, swing1LimitDegrees=45, swing2LimitDegrees=45, twistLimitDegrees=45}
 {bone1Name=tentacleA_03, bone2Name=tentacleA_02, swing1LimitDegrees=45, swing2LimitDegrees=45, twistLimitDegrees=45}
 {bone1Name=tentacleA_04, bone2Name=tentacleA_03, swing1LimitDegrees=45, swing2LimitDegrees=45, twistLimitDegrees=45}
 {bone1Name=tentacleA_05, bone2Name=tentacleA_04, swing1LimitDegrees=45, swing2LimitDegrees=45, twistLimitDegrees=45}
 {bone1Name=tentacleA_06, bone2Name=tentacleA_05, swing1LimitDegrees=45, swing2LimitDegrees=45, twistLimitDegrees=45}
 {bone1Name=tentacleA_07, bone2Name=tentacleA_06, swing1LimitDegrees=45, swing2LimitDegrees=45, twistLimitDegrees=45}
 {bone1Name=tentacleA_08, bone2Name=tentacleA_07, swing1LimitDegrees=45, swing2LimitDegrees=45, twistLimitDegrees=45}
 {bone1Name=tentacleA_09, bone2Name=tentacleA_08, swing1LimitDegrees=45, swing2LimitDegrees=45, twistLimitDegrees=45}
 {bone1Name=tentacleA_10, bone2Name=tentacleA_09, swing1LimitDegrees=45, swing2LimitDegrees=45, twistLimitDegrees=45}
 {bone1Name=fin_top_01, bone2Name=head, swing1LimitDegrees=45, swing2LimitDegrees=45, twistLimitDegrees=30}
 {bone1Name=fin_top_02, bone2Name=fin_top_01, swing1LimitDegrees=45, swing2LimitDegrees=45, twistLimitDegrees=25}
 {bone1Name=fin_top_03, bone2Name=fin_top_02, swing1LimitDegrees=45, swing2LimitDegrees=45, twistLimitDegrees=15}
 {bone1Name=head_tip_01, bone2Name=head, swing1LimitDegrees=45, swing2LimitDegrees=45, twistLimitDegrees=15}
 {bone1Name=head_tip_02, bone2Name=head_tip_01, swing1LimitDegrees=45, swing2LimitDegrees=45, twistLimitDegrees=15}
 {bone1Name=head_tip_03, bone2Name=head_tip_02, swing1LimitDegrees=45, swing2LimitDegrees=45, twistLimitDegrees=15}
]
