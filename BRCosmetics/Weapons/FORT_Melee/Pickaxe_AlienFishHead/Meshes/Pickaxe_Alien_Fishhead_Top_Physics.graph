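# Physics asset for the top mesh of the Alien Fishhead pickaxe, giving its head, main body and tentacle capsule collision and jointed physics.
# A ten-segment tentacle forms a serial constraint chain from a large base capsule that attaches to the main bone, which also links to the head.
# The base joint swings and twists freely and the first segment frees one swing axis, so the tentacle is loosest at its root.
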PhysicsAsset /BRCosmetics/Weapons/FORT_Melee/Pickaxe_AlienFishHead/Meshes/Pickaxe_Alien_Fishhead_Top_Physics
bodies {
 head [{shapeType=Capsule, length=0.75, radius=2.25}]
 main [{shapeType=Capsule, length=1, radius=3}]
 tentacleA_01 [{shapeType=Capsule, length=5.1808, radius=3.7597, rotation=<rot 90 0 0>}]
 tentacleA_02 [{shapeType=Capsule, length=4.3847, radius=4.828, rotation=<rot 90 0 0>}]
 tentacleA_03 [{shapeType=Capsule, length=5.1765, radius=4.6097, rotation=<rot 90 0 0>}]
 tentacleA_04 [{shapeType=Capsule, length=4.5059, radius=4.6916, rotation=<rot 90 0 0>}]
 tentacleA_05 [{shapeType=Capsule, length=5.2919, radius=4.347, rotation=<rot 90 0 0>}]
 tentacleA_06 [{shapeType=Capsule, length=3.3051, radius=4.0811, rotation=<rot 90 0 0>}]
 tentacleA_07 [{shapeType=Capsule, length=2.0525, radius=4.183, rotation=<rot 90 0 0>}]
 tentacleA_08 [{shapeType=Capsule, length=2.1923, radius=4, rotation=<rot 90 0 0>}]
 tentacleA_09 [{shapeType=Capsule, length=2.8373, radius=3.9653, rotation=<rot 90 0 0>}]
 tentacleA_10 [{shapeType=Capsule, length=1, radius=3.755, rotation=<rot 90 0 0>}]
 tentacle_base [{shapeType=Capsule, length=12.505, radius=9.505, center=<0.0001 7.7394 0>, rotation=<rot 90 -90 0.0012>}]
}
constraints [
 {bone1Name=head, bone2Name=main, swing1LimitDegrees=45, swing2LimitDegrees=45, twistLimitDegrees=45}
 {bone1Name=tentacleA_01, bone2Name=tentacle_base, swing1LimitDegrees=45, swing1Motion=Free, swing2LimitDegrees=45, twistLimitDegrees=45}
 {
  bone1Name=tentacle_base
  bone2Name=main
  swing1LimitDegrees=45
  swing1Motion=Free
  swing2LimitDegrees=45
  swing2Motion=Free
  twistLimitDegrees=45
  twistMotion=Free
 }
 {bone1Name=tentacleA_02, bone2Name=tentacleA_01, swing1LimitDegrees=45, swing2LimitDegrees=45, twistLimitDegrees=45}
 {bone1Name=tentacleA_03, bone2Name=tentacleA_02, swing1LimitDegrees=45, swing2LimitDegrees=45, twistLimitDegrees=45}
 {bone1Name=tentacleA_04, bone2Name=tentacleA_03, swing1LimitDegrees=45, swing2LimitDegrees=45, twistLimitDegrees=45}
 {bone1Name=tentacleA_05, bone2Name=tentacleA_04, swing1LimitDegrees=45, swing2LimitDegrees=45, twistLimitDegrees=45}
 {bone1Name=tentacleA_06, bone2Name=tentacleA_05, swing1LimitDegrees=45, swing2LimitDegrees=45, twistLimitDegrees=45}
 {bone1Name=tentacleA_07, bone2Name=tentacleA_06, swing1LimitDegrees=45, swing2LimitDegrees=45, twistLimitDegrees=45}
 {bone1Name=tentacleA_08, bone2Name=tentacleA_07, swing1LimitDegrees=45, swing2LimitDegrees=45, twistLimitDegrees=45}
 {bone1Name=tentacleA_09, bone2Name=tentacleA_08, swing1LimitDegrees=45, swing2LimitDegrees=45, twistLimitDegrees=45}
 {bone1Name=tentacleA_10, bone2Name=tentacleA_09, swing1LimitDegrees=45, swing2LimitDegrees=45, twistLimitDegrees=45}
]
