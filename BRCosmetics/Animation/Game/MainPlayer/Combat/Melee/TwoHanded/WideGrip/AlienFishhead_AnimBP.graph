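# Animation Blueprint for the Alien Fishhead pickaxe, animating the weapon mesh to follow the owning player's movement and melee swings.
# Each update reads the player AnimInstance's moving flag and MeleeAnim curve to drive idle, moving, swinging and swing-to-idle states.
# A blink additive, enabled while moving, feeds a cached pose.
# Two chained RigidBody passes blend over it, fading out when the pickaxe's idle curve equals one.
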AnimBlueprint /BRCosmetics/Animation/Game/MainPlayer/Combat/Melee/TwoHanded/WideGrip/AlienFishhead_AnimBP
target_skeleton=/BRCosmetics/Weapons/FORT_Melee/Pickaxe_AlienFishHead/Meshes/Pickaxe_Alien_Fishhead_Skeleton
graphs [
 AnimGraph {nodes [
   n1 = RigidBody(ComponentPose=Local To Component(LocalPose=Use cached pose 'BaseAnim'()), Alpha=0.85) {
    bOverrideWorldGravity=true
    overridePhysicsAsset=/BRCosmetics/Weapons/FORT_Melee/Pickaxe_AlienFishHead/Meshes/Pickaxe_Alien_Fishhead_Physics
    alphaBoolBlend {blendOption=Linear}
    componentAppliedLinearAccClamp=<1000 1000 1000>
    componentLinearAccScale=<0.8 0.8 0.8>
    componentLinearVelScale=<0.8 0.8 0.8>
    overrideWorldGravity {z=-980}
   }
   n2 = RigidBody(ComponentPose=@n1, Alpha=0.25) {
    bOverrideWorldGravity=true
    overridePhysicsAsset=/BRCosmetics/Weapons/FORT_Melee/Pickaxe_AlienFishHead/Meshes/Pickaxe_Alien_Fishhead_Top_Physics
    alphaBoolBlend {blendOption=Linear}
    componentAppliedLinearAccClamp=<1000 1000 1000>
    componentLinearAccScale=<0.8 0.8 0.8>
    componentLinearVelScale=<0.8 0.8 0.8>
    overrideWorldGravity {z=-980}
   }
   n3 = Component To Local(ComponentPose=@n2)
   n4 = Use cached pose 'BaseAnim'()
   n5 = Get DynamicsAlpha() -> (DynamicsAlpha:double)
   Output Pose(Result=Slot 'DefaultSlot'(Source=Two Way Blend(A=@n4, B=@n3, Alpha=@n5) {blendNode {alphaBoolBlend {blendOption=Linear}}})) {layerGroup=None}
   AnimGraphNode_StateMachine_0 = Locomotion() {is=title AnimGraphNode_StateMachine}
   Save cached pose 'BaseAnim'(Pose=Apply Additive!(
     Base=@AnimGraphNode_StateMachine_0
     Additive="Sequence Player (additive) 'AlienFishhead_Blink_200Frames_PICKAXE'"() {sequence=&/AlienFishhead_Blink_200Frames_PICKAXE}
     bAlphaBoolEnabled:bool=Get Moving2d() -> (Moving2d:bool)
    ) -> (Pose) {alphaInputType=Bool, alphaBoolBlend {blendOption=Linear}}) {cacheName=BaseAnim}
  ]}
 Locomotion {
  entry=idle
  path=AnimGraph.AnimGraphNode_StateMachine_0.Locomotion
  ids=id
  nodes [
   AnimStateNode_0 = idle() {is=title AnimStateNode}
   AnimStateNode_1 = swinging() {is=title AnimStateNode}
   AnimStateNode_2 = moving() {is=title AnimStateNode}
   AnimStateNode_3 = swing_to_idle() {is=title AnimStateNode}
   AnimStateTransitionNode_0 = AnimStateTransitionNode()
   AnimStateTransitionNode_2 = AnimStateTransitionNode()
   AnimStateTransitionNode_3 = AnimStateTransitionNode()
   AnimStateTransitionNode_4 = AnimStateTransitionNode() {blendMode=Linear, crossfadeDuration=0.025}
   AnimStateTransitionNode_5 = AnimStateTransitionNode() {bAutomaticRuleBasedOnSequencePlayerInState=true, blendMode=Linear, crossfadeDuration=0.1}
   AnimStateTransitionNode_6 = AnimStateTransitionNode()
   AnimStateTransitionNode_7 = AnimStateTransitionNode() {bAutomaticRuleBasedOnSequencePlayerInState=true, crossfadeDuration=0.1, priorityOrder=2}
  ]
  links [
   {
    dst=moving
    rule=AnimGraph.AnimGraphNode_StateMachine_0.Locomotion.AnimStateTransitionNode_2.Transition
    src=idle
    via=AnimStateTransitionNode_2
   }
   {
    dst=swinging
    rule=AnimGraph.AnimGraphNode_StateMachine_0.Locomotion.AnimStateTransitionNode_0.Transition
    src=idle
    via=AnimStateTransitionNode_0
   }
   {
    dst=idle
    rule=AnimGraph.AnimGraphNode_StateMachine_0.Locomotion.AnimStateTransitionNode_3.Transition
    src=moving
    via=AnimStateTransitionNode_3
   }
   {
    dst=swinging
    rule=AnimGraph.AnimGraphNode_StateMachine_0.Locomotion.AnimStateTransitionNode_6.Transition
    src=moving
    via=AnimStateTransitionNode_6
   }
   {
    dst=idle
    rule=AnimGraph.AnimGraphNode_StateMachine_0.Locomotion.AnimStateTransitionNode_5.Transition
    src=swing_to_idle
    via=AnimStateTransitionNode_5
   }
   {
    dst=moving
    rule=AnimGraph.AnimGraphNode_StateMachine_0.Locomotion.AnimStateTransitionNode_7.Transition
    src=swing_to_idle
    via=AnimStateTransitionNode_7
   }
   {
    dst=swing_to_idle
    rule=AnimGraph.AnimGraphNode_StateMachine_0.Locomotion.AnimStateTransitionNode_4.Transition
    src=swinging
    via=AnimStateTransitionNode_4
   }
  ]
 }
 idle {
  path=AnimGraph.AnimGraphNode_StateMachine_0.Locomotion.AnimStateNode_0.idle
  nodes [Output Animation Pose(Result=Apply Additive(
     Base=Sequence Player 'AlienFishhead_Idle_PICKAXE'() {sequence=&/AlienFishhead_Idle_PICKAXE}
     Additive="Sequence Player (additive) 'AlienFishhead_ClipFix_Pose_PICKAXE'"() {sequence=&/AlienFishhead_ClipFix_Pose_PICKAXE}
     bAlphaBoolEnabled=True
    ) {alphaInputType=Bool, alphaBoolBlend {blendOption=Linear}})]
 }
 swinging {
  path=AnimGraph.AnimGraphNode_StateMachine_0.Locomotion.AnimStateNode_1.swinging
  nodes [Output Animation Pose(Result=Sequence Player 'AlienFishhead_Swing_Pose_PICKAXE'() {sequence=&/AlienFishhead_Swing_Pose_PICKAXE})]
 }
 moving {
  path=AnimGraph.AnimGraphNode_StateMachine_0.Locomotion.AnimStateNode_2.moving
  nodes [
   Sequence Player 'AlienFishhead_Jog_Pose_PICKAXE'() {sequence=&/AlienFishhead_Jog_Pose_PICKAXE}
   Output Animation Pose(Result=Apply Additive(
     Base=Sequence Player 'AlienFishhead_Jog_PICKAXE'() {sequence=&/AlienFishhead_Jog_PICKAXE}
     Additive="Sequence Player (additive) 'AlienFishhead_ClipFix_Pose_PICKAXE'"() {sequence=&AlienFishhead_ClipFix_Pose_PICKAXE}
     bAlphaBoolEnabled=True
    ) {alphaInputType=Bool, alphaBoolBlend {blendOption=Linear}})
  ]
 }
 swing_to_idle {
  path=AnimGraph.AnimGraphNode_StateMachine_0.Locomotion.AnimStateNode_3.swing_to_idle
  nodes [
   n1 = Sequence Player 'AlienFishhead_Swing_to_Idle_PICKAXE'() {bLoopAnimation=false, sequence=&/AlienFishhead_Swing_to_Idle_PICKAXE}
   Output Animation Pose(Result=Apply Additive(
     Base=@n1
     Additive="Sequence Player (additive) 'AlienFishhead_ClipFix_Pose_PICKAXE'"() {sequence=&AlienFishhead_ClipFix_Pose_PICKAXE}
     bAlphaBoolEnabled=True
    ) {alphaInputType=Bool, alphaBoolBlend {blendOption=Linear}})
  ]
 }
 Transition {
  path=AnimGraph.AnimGraphNode_StateMachine_0.Locomotion.AnimStateTransitionNode_0.Transition
  nodes [
   Result(bCanEnterTransition=Get MeleeAnimPlaying() -> (MeleeAnimPlaying:bool))
   Get BlockAnimTransition() -> (BlockAnimTransition:bool)
  ]
 }
 Transition {
  path=AnimGraph.AnimGraphNode_StateMachine_0.Locomotion.AnimStateTransitionNode_2.Transition
  nodes [
   Result(bCanEnterTransition=Get Moving2d())
   Get BlockAnimTransition()
  ]
 }
 Transition {
  path=AnimGraph.AnimGraphNode_StateMachine_0.Locomotion.AnimStateTransitionNode_3.Transition
  nodes [
   Result(bCanEnterTransition=NOT Boolean(A=Get Moving2d()))
   Get BlockAnimTransition()
  ]
 }
 Transition {
  path=AnimGraph.AnimGraphNode_StateMachine_0.Locomotion.AnimStateTransitionNode_4.Transition
  nodes [Result(bCanEnterTransition=NOT Boolean(A=Get MeleeAnimPlaying()))]
 }
 Transition {
  path=AnimGraph.AnimGraphNode_StateMachine_0.Locomotion.AnimStateTransitionNode_5.Transition
  nodes [Result(bCanEnterTransition=NOT Boolean(A=Get Moving2d()))]
 }
 Transition {
  path=AnimGraph.AnimGraphNode_StateMachine_0.Locomotion.AnimStateTransitionNode_6.Transition
  nodes [Result(bCanEnterTransition=Get MeleeAnimPlaying())]
 }
 Transition {
  path=AnimGraph.AnimGraphNode_StateMachine_0.Locomotion.AnimStateTransitionNode_7.Transition
  nodes [Result(bCanEnterTransition=Get Moving2d())]
 }
 EventGraph {nodes [
   n1 = Event BlueprintUpdateAnimation()
   > Set Time(Time:double=@n1.DeltaTimeX) -> (Output_Get:double)
   > n2 = Cast To FortWeaponPickaxeAthena!(execute, Object=GetOwningActor()) -> (then, CastFailed, AsFort Weapon Pickaxe Athena:Fort Weapon Pickaxe Athena Object Reference) {is=title K2Node_DynamicCast}
   n3 = GetOwner(self:Actor Object Reference=@n2)
   > Is Valid(InputObject=@n3)
   n4 = GetComponentByClass!(self:Actor Object Reference=@n3, ComponentClass:Actor Component Class Reference=/Script/Engine.SkeletalMeshComponent) -> (ReturnValue:Skeletal Mesh Component Object Reference)
   > Is Valid(InputObject=@n4)
   > n5 = Cast To FortPlayerAnimInstance(Object=GetAnimInstance(self=@n4))
   > Set MeleeAnimPlaying(MeleeAnimPlaying:bool="float >= float"(A=GetCurveValue(self=@n5, CurveName=MeleeAnim), B=1) {is=title K2Node_CallFunction}) -> (Output_Get:bool)
   > Set Moving2d(Moving2d:bool=Get bIsMoving2D!(self:Fort Player Anim Instance Object Reference=@n5) -> (bIsMoving2D)) -> (Output_Get:bool)
   > n6 = Branch(Condition="Equal (Float)"!(A:double=GetCurveValue(CurveName=idle), B:double=1) -> (ReturnValue:bool) {is=title K2Node_CallFunction})
   > Set DynamicsAlpha(DynamicsAlpha:double=FInterpTo(Current=Get DynamicsAlpha(), Target=0, DeltaTime=Get Time() -> (Time:double), InterpSpeed=3)) -> (Output_Get:double)
   Set DynamicsAlpha(execute=@n6.else, DynamicsAlpha=FInterpTo(Current=Get DynamicsAlpha(), Target=0.6, DeltaTime=Get Time(), InterpSpeed=3))
  ]}
]
variables [
 BlockAnimTransition {type=boolean}
 DynamicsAlpha {type=number}
 MeleeAnimPlaying {type=boolean}
 Moving2d {type=boolean}
 Time {type=number}
]
