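# Animation Blueprint for the Pirate Octopus pickaxe, animating the weapon as its owning player idles and swings.
# Each update reads the owner's melee curve and movement flag to help drive idle, moving, swinging and swing-to-idle states.
# The state pose feeds a linked Pirate_Octopus_Axe_AnimBP, then DefaultSlot.
# BlockAnimTransition gates several transitions but is never set here; DynamicsAlpha eases with the idle curve but is unused here.
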AnimBlueprint /BRCosmetics/Animation/Game/MainPlayer/Combat/Melee/TwoHanded/WideGrip/Pirate_Octopus_Axe_AnimBP_Base
target_skeleton=/BRCosmetics/Weapons/FORT_Melee/Pickaxe_PirateOctopus/Meshes/Pirate_Octopus_Axe_Skeleton
graphs [
 AnimGraph {nodes [
   AnimGraphNode_StateMachine_0 = Locomotion() {is=title AnimGraphNode_StateMachine}
   AnimGraphNode_SubInstance_0 = Pirate_Octopus_Axe_AnimBP!(
    InPose:Pose Link Structure=@AnimGraphNode_StateMachine_0
    bIsRiding:bool=false
    bIsSkyDivingFromBus:bool=false
    bParachuteOpenedAndStandingUpright:bool=false
    bShouldDisplayFrontendReactivity:bool=false
    bIsPreviewingItemInFrontEnd:bool=false
    bIsOnPawn:bool=false
    bIsInAir:bool=false
    bVehicleSeatForcesCrouch:bool=false
    bIsMotorcycleDriver:bool=false
    bIsMotorcyclePassenger:bool=false
    bHasMotorcyclePassenger:bool=false
    bIsTacticalSprinting:bool=false
    bIsLeftArmHidden:bool=false
    bIsRightArmHidden:bool=false
    bIsLeftLegHidden:bool=false
    bIsRightLegHidden:bool=false
    bIsNeckHeadHidden:bool=false
    VehicleSeatIndex:int=0
    RidingSpeed2D:float=0
    SignificanceLOD:int=0
    PlayerGliderType:EGliderType Enum=HangGlider
    bPbBlendEnabled:bool=false
   ) -> (Pose:Pose Link Structure) {
    is=title AnimGraphNode_LinkedAnimGraph
    instanceClass=/BRCosmetics/Weapons/FORT_Melee/Pickaxe_PirateOctopus/Meshes/Pirate_Octopus_Axe_AnimBP.Pirate_Octopus_Axe_AnimBP_C
   }
   Output Pose(Result=Slot 'DefaultSlot'(Source=@AnimGraphNode_SubInstance_0)) {layerGroup=None}
  ]}
 Locomotion {
  entry=idle
  path=AnimGraph.AnimGraphNode_StateMachine_0.Locomotion
  ids=id
  nodes [
   AnimStateNode_0 = idle() {is=title AnimStateNode}
   AnimStateNode_1 = swinging() {is=title AnimStateNode}
   AnimStateNode_2 = moving() {is=title AnimStateNode}
   AnimStateNode_3 = swing_to_idle() {is=title AnimStateNode}
   AnimStateTransitionNode_0 = AnimStateTransitionNode()
   AnimStateTransitionNode_2 = AnimStateTransitionNode()
   AnimStateTransitionNode_3 = AnimStateTransitionNode()
   AnimStateTransitionNode_4 = AnimStateTransitionNode() {blendMode=Linear, crossfadeDuration=0.025}
   AnimStateTransitionNode_5 = AnimStateTransitionNode() {bAutomaticRuleBasedOnSequencePlayerInState=true, blendMode=Linear, crossfadeDuration=0.1}
   AnimStateTransitionNode_6 = AnimStateTransitionNode()
   AnimStateTransitionNode_7 = AnimStateTransitionNode() {bAutomaticRuleBasedOnSequencePlayerInState=true, crossfadeDuration=0.1, priorityOrder=2}
  ]
  links [
   {
    dst=moving
    rule=AnimGraph.AnimGraphNode_StateMachine_0.Locomotion.AnimStateTransitionNode_2.Transition
    src=idle
    via=AnimStateTransitionNode_2
   }
   {
    dst=swinging
    rule=AnimGraph.AnimGraphNode_StateMachine_0.Locomotion.AnimStateTransitionNode_0.Transition
    src=idle
    via=AnimStateTransitionNode_0
   }
   {
    dst=idle
    rule=AnimGraph.AnimGraphNode_StateMachine_0.Locomotion.AnimStateTransitionNode_3.Transition
    src=moving
    via=AnimStateTransitionNode_3
   }
   {
    dst=swinging
    rule=AnimGraph.AnimGraphNode_StateMachine_0.Locomotion.AnimStateTransitionNode_6.Transition
    src=moving
    via=AnimStateTransitionNode_6
   }
   {
    dst=idle
    rule=AnimGraph.AnimGraphNode_StateMachine_0.Locomotion.AnimStateTransitionNode_5.Transition
    src=swing_to_idle
    via=AnimStateTransitionNode_5
   }
   {
    dst=moving
    rule=AnimGraph.AnimGraphNode_StateMachine_0.Locomotion.AnimStateTransitionNode_7.Transition
    src=swing_to_idle
    via=AnimStateTransitionNode_7
   }
   {
    dst=swing_to_idle
    rule=AnimGraph.AnimGraphNode_StateMachine_0.Locomotion.AnimStateTransitionNode_4.Transition
    src=swinging
    via=AnimStateTransitionNode_4
   }
  ]
 }
 idle {
  path=AnimGraph.AnimGraphNode_StateMachine_0.Locomotion.AnimStateNode_0.idle
  nodes [Output Animation Pose(Result=Sequence Player 'Pirate_Octopus_Axe_Idle_PICKAXE'() {sequence=&/Pirate_Octopus_Axe_Idle_PICKAXE})]
 }
 swinging {
  path=AnimGraph.AnimGraphNode_StateMachine_0.Locomotion.AnimStateNode_1.swinging
  nodes [Output Animation Pose(Result=Sequence Player 'Pirate_Octopus_Axe_Swing_Pose_PICKAXE'() {sequence=&/Pirate_Octopus_Axe_Swing_Pose_PICKAXE})]
 }
 moving {
  path=AnimGraph.AnimGraphNode_StateMachine_0.Locomotion.AnimStateNode_2.moving
  nodes [Output Animation Pose(Result=Sequence Player 'Pirate_Octopus_Axe_Idle_PICKAXE'() {sequence=&Pirate_Octopus_Axe_Idle_PICKAXE})]
 }
 swing_to_idle {
  path=AnimGraph.AnimGraphNode_StateMachine_0.Locomotion.AnimStateNode_3.swing_to_idle
  nodes [Output Animation Pose(Result=Sequence Player 'Pirate_Octopus_Axe_Idle_PICKAXE'() {sequence=&Pirate_Octopus_Axe_Idle_PICKAXE})]
 }
 Transition {
  path=AnimGraph.AnimGraphNode_StateMachine_0.Locomotion.AnimStateTransitionNode_0.Transition
  nodes [Result(bCanEnterTransition=Get MeleeAnimPlaying() -> (MeleeAnimPlaying:bool))]
 }
 Transition {
  path=AnimGraph.AnimGraphNode_StateMachine_0.Locomotion.AnimStateTransitionNode_2.Transition
  nodes [
   Result(bCanEnterTransition=Get BlockAnimTransition() -> (BlockAnimTransition:bool))
   Get Moving2d() -> (Moving2d:bool)
  ]
 }
 Transition {
  path=AnimGraph.AnimGraphNode_StateMachine_0.Locomotion.AnimStateTransitionNode_3.Transition
  nodes [
   Result(bCanEnterTransition=NOT Boolean(A=Get Moving2d()))
   Get BlockAnimTransition()
  ]
 }
 Transition {
  path=AnimGraph.AnimGraphNode_StateMachine_0.Locomotion.AnimStateTransitionNode_4.Transition
  nodes [Result(bCanEnterTransition=NOT Boolean(A=Get MeleeAnimPlaying()))]
 }
 Transition {
  path=AnimGraph.AnimGraphNode_StateMachine_0.Locomotion.AnimStateTransitionNode_5.Transition
  nodes [
   Result(bCanEnterTransition=NOT Boolean(A=Get Moving2d()))
   Get BlockAnimTransition()
  ]
 }
 Transition {
  path=AnimGraph.AnimGraphNode_StateMachine_0.Locomotion.AnimStateTransitionNode_6.Transition
  nodes [
   Result(bCanEnterTransition=Get BlockAnimTransition())
   Get MeleeAnimPlaying()
  ]
 }
 Transition {
  path=AnimGraph.AnimGraphNode_StateMachine_0.Locomotion.AnimStateTransitionNode_7.Transition
  nodes [
   Result(bCanEnterTransition=Get BlockAnimTransition())
   Get Moving2d()
  ]
 }
 EventGraph {nodes [
   n1 = Event BlueprintUpdateAnimation()
   > Set Time(Time:double=@n1.DeltaTimeX) -> (Output_Get:double)
   > n2 = Cast To FortWeaponPickaxeAthena!(execute, Object=GetOwningActor()) -> (then, CastFailed, AsFort Weapon Pickaxe Athena:Fort Weapon Pickaxe Athena Object Reference) {is=title K2Node_DynamicCast}
   n3 = GetOwner(self:Actor Object Reference=@n2)
   > Is Valid(InputObject=@n3)
   n4 = GetComponentByClass!(self:Actor Object Reference=@n3, ComponentClass:Actor Component Class Reference=/Script/Engine.SkeletalMeshComponent) -> (ReturnValue:Skeletal Mesh Component Object Reference)
   > Is Valid(InputObject=@n4)
   > n5 = Cast To FortPlayerAnimInstance(Object=GetAnimInstance(self=@n4))
   > Set MeleeAnimPlaying(MeleeAnimPlaying:bool="float >= float"(A=GetCurveValue(self=@n5, CurveName=MeleeAnim), B=1) {is=title K2Node_CallFunction}) -> (Output_Get:bool)
   > Set Moving2d(Moving2d:bool=Get bIsMoving2D!(self:Fort Player Anim Instance Object Reference=@n5) -> (bIsMoving2D)) -> (Output_Get:bool)
   > n6 = Branch(Condition="Equal (Float)"!(A:double=GetCurveValue(CurveName=idle), B:double=1) -> (ReturnValue:bool) {is=title K2Node_CallFunction})
   > Set DynamicsAlpha(DynamicsAlpha:double=FInterpTo(Current=Get DynamicsAlpha() -> (DynamicsAlpha:double), Target=0, DeltaTime=Get Time() -> (Time:double), InterpSpeed=3)) -> (Output_Get:double)
   Set DynamicsAlpha(execute=@n6.else, DynamicsAlpha=FInterpTo(Current=Get DynamicsAlpha(), Target=0.6, DeltaTime=Get Time(), InterpSpeed=3))
  ]}
]
variables [
 BlockAnimTransition {type=boolean}
 DynamicsAlpha {type=number}
 MeleeAnimPlaying {type=boolean}
 Moving2d {type=boolean}
 Time {type=number}
]
